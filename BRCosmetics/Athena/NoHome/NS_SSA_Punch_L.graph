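# Niagara effect for the SeaSaltArmor left punch (by its name), combining brief mesh jet flashes with smoke puffs.
# Jets bursts eight meshes, using spawn group index to pin each to a socket on User.SeaSaltArmor_L, flipping rotation between halves, then scaling and fading them.
# Smoke_01 and Smoke_02 briefly spawn grey sprites offset above and below the system origin, animated through SubUV frames, size curves and dynamic material parameters.
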NiagaraSystem /BRCosmetics/Athena/NoHome/NS_SSA_Punch_L
dependencies {usedDataInterfaces [&NiagaraDataInterfaceCurve], usedDynamicInputs [], usedModules [&SystemState]}
emitters {
 Jets {
  emitterUpdateScript [
   &EmitterState(Life Cycle Mode=::NewEnumerator1, Loop Behavior=::NewEnumerator1, Loop Duration=1, UseLoopDelay=false)
   &SpawnBurst_Instantaneous(Spawn Count=8, Use Spawn Probability=false)
  ]
  particleSpawnScript [
   &FN_SetSpawnGroup_v2(Number of Spawn Groups=8)
   /Niagara/Modules/Spawn/Initialization/InitializeParticle(
    Color=<rgba 0.6 0.8 1 1>
    Lifetime=0.25
    Mesh Scale=&VectorFromFloat(Value=0.15)
    Module.Write Color=true
    Module.Write Lifetime=true
    Module.Write Mass=true
    Module.Write Material Random=true
    Module.Write Position=true
    Module.Write Scale=true
    Module.Write SpriteRotation=false
    Module.Write SpriteSize=true
    Sprite Size=<5 5>
   )
   /Niagara/Modules/Spawn/Orientation/V2/InitialMeshOrientation(
    Rotation=&Lerp_Vector(A=<0 -0.12 0>, B=<0 0.12 0>, Alpha=HlslExpression() {hlslExpression="Particles.SpawnGroup < 4 ? 1:0"})
    Use Orientation Vector=false
    Use Rotation Vector=true
   )
  ]
  particleUpdateScript [
   &ParticleState()
   &SkeletalMeshLocation(
    Apply Sampled Orientation to Mesh Particles=true
    Bone / Socket Index=HlslExpression() {hlslExpression=Particles.SpawnGroup}
    Inherit Velocity=false
    MaskSpawnGroup=false
    Mesh Sampling Type=::NewEnumerator5
    Skeletal Mesh=$User.SeaSaltArmor_L
    Socket Sampling Mode=::NewEnumerator5
   )
   &ScaleColor(
    Scale Alpha=&FloatFromCurve(FloatCurve=DataInterface() {propertyValues {Curve {keys [<at 0 1>,<at 1 0>]}}})
    Scale RGB=&VectorFromFloat(Value=&FloatFromCurve(FloatCurve=DataInterface() {propertyValues {Curve {keys [<at 0 3>,<at 1 1>]}}}))
    ScaleA=true
    ScaleRGB=true
   )
   &ScaleMeshSize(Scale Factor=&VectorFromFloat(Value=&FloatFromCurve(FloatCurve=DataInterface() {propertyValues {Curve {keys [
          {interpMode=RCIM_Cubic, leaveTangent=5.1135, tangentMode=RCTM_Break, time=0, value=1}
          {arriveTangent=-1, interpMode=RCIM_Cubic, leaveTangent=-1, tangentMode=RCTM_Break, time=0.2936, value=2}
          {arriveTangent=-19.3278, interpMode=RCIM_Cubic, tangentMode=RCTM_Break, time=0.5, value=0}
         ]}}})))
  ]
  rendererClasses [&NiagaraMeshRendererProperties]
  renderers [{rendererClass=&NiagaraMeshRendererProperties, rendererIndex=0}]
 }
 Smoke_01 {
  emitterUpdateScript [
   &EmitterState(Life Cycle Mode=::NewEnumerator1, Loop Behavior=::NewEnumerator1, Loop Duration=0.085, UseLoopDelay=false)
   &SpawnRate(SpawnRate=350, Use Spawn Probability=false)
  ]
  particleSpawnScript [
   &InitializeParticle(
    Color=&UniformRangedLinearColor(Minimum=<rgba 0.4323 0.4323 0.4323 0.35>, Maximum=<rgba 0.5255 0.5255 0.5255 0.5>)
    Lifetime=&UniformRangedFloat(Minimum=0.125, Maximum=2)
    Module.Write Color=true
    Module.Write Lifetime=true
    Module.Write Mass=true
    Module.Write Material Random=true
    Module.Write Position=false
    Module.Write Scale=false
    Module.Write SpriteRotation=true
    Module.Write SpriteSize=true
    Position=&SimulationPosition()
    Sprite Rotation=&UniformRangedFloat(Minimum=0, Maximum=360)
    Sprite Size=&Vector2DFromFloat(Value=&UniformRangedFloat(Minimum=12, Maximum=25))
   )
   &AccelerationForce(Acceleration=&UniformRangedVector(Minimum=<0 0 0>, Maximum=<0 0 50>)) {enabled=false}
   &AddVelocity(Velocity=&Multiply_Vector(A=&UniformRangedVector(Minimum=<-1 -1 1>, Maximum=<1 1 1>), B=<35 25 20>)) {enabled=false}
   &ApplyInitialForces()
   &SystemLocation(MaskSpawnGroup=false, Offset=<0 0 5>)
  ]
  particleUpdateScript [
   &ParticleState()
   &Drag(Drag=0.7, Use Linear Drag=true, Use Rotational Drag=false)
   &SolveForcesAndVelocity(Clamp Velocity=false, Limit Acceleration=false)
   &ScaleSpriteSize(Scale Factor=&Vector2DFromFloat(Value=&FloatFromCurve(FloatCurve=DataInterface() {propertyValues {Curve {keys [
          {arriveTangent=1.491, interpMode=RCIM_Cubic, tangentMode=RCTM_Break, time=-0.0046, value=0}
          {interpMode=RCIM_Cubic, tangentMode=RCTM_Break, time=0.25, value=1.75}
          {interpMode=RCIM_Cubic, tangentMode=RCTM_Break, time=0.5, value=0}
         ]}}})))
   /SetVariables(Particles.SubImageIndex=&FloatFromCurve(FloatCurve=DataInterface() {propertyValues {Curve {keys [
         {arriveTangent=29.8623, interpMode=RCIM_Cubic, leaveTangent=29.8623, tangentMode=RCTM_User, time=-0.0046, value=5.0556}
         <at 1.0011 22.1689 RCIM_Cubic>
        ]}}})) {isSetParameters=true, module=SetVariables_#1}
   &DynamicMaterialParameters(
    Index 0 Param 1=&FloatFromCurve(FloatCurve=DataInterface() {propertyValues {Curve {keys [<at -0.0046 1.0113 RCIM_Cubic>,<at 1.0011 0.002>]}}})
    Index 0 Param 4=&FloatFromCurve(FloatCurve=DataInterface() {propertyValues {Curve {keys [<at 0 0 RCIM_Cubic>,<at 1 0.5>]}}})
    Module.Param0WriteEnabled=true
    Module.Param1WriteEnabled=false
    Module.Param2WriteEnabled=false
    Module.Param3WriteEnabled=true
   ) {module=DynamicMatParams}
   &CameraOffset(Camera Offset Amount=50) {enabled=false}
   &ScaleColor(
    Scale Alpha=&FloatFromCurve(
     FloatCurve=DataInterface() {propertyValues {Curve {keys [
         {arriveTangent=-1.7104, interpMode=RCIM_Cubic, leaveTangent=-1.7104, tangentMode=RCTM_User, time=0.1246, value=0.5559}
         {arriveTangent=0.0642, interpMode=RCIM_Cubic, leaveTangent=0.0642, tangentMode=RCTM_User, time=1.0011, value=0.002}
        ]}}}
     Scale Curve=0.1
    )
    ScaleA=true
    ScaleRGB=false
   )
  ]
  rendererClasses [&NiagaraSpriteRendererProperties]
  renderers [{rendererClass=&NiagaraSpriteRendererProperties}]
 }
 Smoke_02 {
  emitterUpdateScript [
   &EmitterState(Life Cycle Mode=::NewEnumerator1, Loop Behavior=::NewEnumerator1, Loop Duration=0.085, UseLoopDelay=false)
   &SpawnRate(SpawnRate=350, Use Spawn Probability=false)
  ]
  particleSpawnScript [
   &InitializeParticle(
    Color=&UniformRangedLinearColor(Minimum=<rgba 0.4323 0.4323 0.4323 0.35>, Maximum=<rgba 0.5255 0.5255 0.5255 0.5>)
    Lifetime=&UniformRangedFloat(Minimum=0.125, Maximum=2)
    Module.Write Color=true
    Module.Write Lifetime=true
    Module.Write Mass=true
    Module.Write Material Random=true
    Module.Write Position=false
    Module.Write Scale=false
    Module.Write SpriteRotation=true
    Module.Write SpriteSize=true
    Position=&SimulationPosition()
    Sprite Rotation=&UniformRangedFloat(Minimum=0, Maximum=360)
    Sprite Size=&Vector2DFromFloat(Value=&UniformRangedFloat(Minimum=12, Maximum=25))
   )
   &AccelerationForce(Acceleration=&UniformRangedVector(Minimum=<0 0 0>, Maximum=<0 0 50>)) {enabled=false}
   &AddVelocity(Velocity=&Multiply_Vector(A=&UniformRangedVector(Minimum=<-1 -1 1>, Maximum=<1 1 1>), B=<35 25 -20>)) {enabled=false}
   &ApplyInitialForces()
   &SystemLocation(MaskSpawnGroup=false, Offset=<0 0 -5>)
  ]
  particleUpdateScript [
   &ParticleState()
   &Drag(Drag=0.7, Use Linear Drag=true, Use Rotational Drag=false)
   &SolveForcesAndVelocity(Clamp Velocity=false, Limit Acceleration=false)
   &ScaleSpriteSize(Scale Factor=&Vector2DFromFloat(Value=&FloatFromCurve(FloatCurve=DataInterface() {propertyValues {Curve {keys [
          {arriveTangent=1.491, interpMode=RCIM_Cubic, tangentMode=RCTM_Break, time=-0.0046, value=0}
          {interpMode=RCIM_Cubic, tangentMode=RCTM_Break, time=0.25, value=1.75}
          {interpMode=RCIM_Cubic, tangentMode=RCTM_Break, time=0.5, value=0}
         ]}}})))
   /SetVariables(Particles.SubImageIndex=&FloatFromCurve(FloatCurve=DataInterface() {propertyValues {Curve {keys [
         {arriveTangent=29.8623, interpMode=RCIM_Cubic, leaveTangent=29.8623, tangentMode=RCTM_User, time=-0.0046, value=5.0556}
         <at 1.0011 22.1689 RCIM_Cubic>
        ]}}})) {isSetParameters=true, module=SetVariables_#1}
   &DynamicMaterialParameters(
    Index 0 Param 1=&FloatFromCurve(FloatCurve=DataInterface() {propertyValues {Curve {keys [<at -0.0046 1.0113 RCIM_Cubic>,<at 1.0011 0.002>]}}})
    Index 0 Param 4=&FloatFromCurve(FloatCurve=DataInterface() {propertyValues {Curve {keys [<at 0 0 RCIM_Cubic>,<at 1 0.5>]}}})
    Module.Param0WriteEnabled=true
    Module.Param1WriteEnabled=false
    Module.Param2WriteEnabled=false
    Module.Param3WriteEnabled=true
   ) {module=DynamicMatParams}
   &CameraOffset(Camera Offset Amount=50) {enabled=false}
   &ScaleColor(
    Scale Alpha=&FloatFromCurve(
     FloatCurve=DataInterface() {propertyValues {Curve {keys [
         {arriveTangent=-1.729, interpMode=RCIM_Cubic, leaveTangent=-1.729, tangentMode=RCTM_User, time=0.1259, value=0.5101}
         {arriveTangent=0.0642, interpMode=RCIM_Cubic, leaveTangent=0.0642, tangentMode=RCTM_User, time=1.0011, value=0.002}
        ]}}}
     Scale Curve=0.1
    )
    ScaleA=true
    ScaleRGB=false
   )
  ]
  rendererClasses [&NiagaraSpriteRendererProperties]
  renderers [{rendererClass=&NiagaraSpriteRendererProperties}]
 }
}
summary {
 systemName=NS_SSA_Punch_L
 emitters [
  {emitterName=Smoke_02, rendererClasses [&NiagaraSpriteRendererProperties]}
  {emitterName=Jets, rendererClasses [&NiagaraMeshRendererProperties]}
  {emitterName=Smoke_01, rendererClasses [&NiagaraSpriteRendererProperties]}
 ]
 userVariables [
  User.bIsFrontEndPreview {type=NiagaraBool}
  User.FlameSize {type=NiagaraFloat, defaultValue=0.4, description=""}
  User.SeaSaltArmor_L {
   type=NiagaraDataInterfaceSkeletalMesh
   defaultValue {
    struct=NiagaraExt_VariableValue_DataInterface
    value {dataInterface=&self:NiagaraDataInterfaceSkeletalMesh_1}
   }
  }
  User.SizeScale {type=NiagaraFloat, defaultValue=0, description=""}
 ]
}
